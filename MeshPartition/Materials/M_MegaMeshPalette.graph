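# M_MegaMeshPalette is a base material in the MeshPartition content, named as the palette for mega mesh surfaces.
# It uses Material Attributes with the Default Lit shading model.
# Its graph holds only a Material Attribute Layers node and no shading logic of its own, so it serves as a shell for layered materials.
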Material /MeshPartition/Materials/M_MegaMeshPalette
graph {properties {bUseMaterialAttributes=true, usedShadingModels=Default Lit}, nodes [MaterialAttributeLayers!(None)]}
