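# MF_PaintGrid is a material function that computes an antialiased grid-line mask from a Float2 Signal, using Period for spacing and Width as a fraction of that spacing.
# Screen-space derivatives (DDX of R, DDY of G) size a filter footprint, over which the stripe pattern is analytically integrated with Floor and Frac to give line coverage.
# The final Max takes R twice, so only first-axis stripes reach the output.
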MaterialFunction /MeshPartition/Materials/MF_PaintGrid
graph {nodes [
  n1 = FunctionInput() {description=Input signal on which the stripe value is computed, inputName=Signal, inputType=FunctionInput_Vector2}
  n2 = FunctionInput() {
   bUsePreviewValueAsDefault=true
   description=Width of the line painted expressed as a percentage of the GridPeriod
   inputName=Width
   inputType=FunctionInput_Scalar
   previewValue=<0.05 0 0 1>
  }
  n3 = FunctionInput() {
   bUsePreviewValueAsDefault=true
   description="Distance between each interval of the grid (in Signal unit)"
   inputName=Period
   inputType=FunctionInput_Scalar
   previewValue=<0.1 0 0 1>
  }
  n4 = Multiply(A=Multiply(B=@n2) {constA=0.5, constB=0.5}, B=@n3) {constA=0.5, constB=0.5}
  n5 = MaterialFunctionCall(Float2=@n1) {materialFunction=&BreakOutFloat2Components}
  n6 = DDX(Value=@n5.R)
  n7 = DDY(Value=@n5.G)
  n8 = Divide(A=Add(A=Abs(None=@n6), B=Abs(None=@n7)), B=@n3)
  n9 = Subtract(A=Divide(A=Subtract(A=@n1, B=@n4), B=@n3), B=Multiply(A=@n8) {constB=0.5})
  n10 = Add(A=@n9, B=@n8)
  n11 = Floor(None=@n10)
  n12 = Divide(A=@n2) {constB=1}
  n13 = OneMinus(None=@n12)
  n14 = Subtract(A=Frac(None=@n10), B=@n13)
  n15 = Floor(None=@n9)
  n16 = Subtract(A=Frac(None=@n9), B=@n13)
  n17 = Subtract(A=Add(A=Multiply(A=@n11, B=@n12), B=Max(A=@n14) {constB=0}), B=Add(A=Multiply(A=@n15, B=@n12), B=Max(A=@n16) {constB=0}))
  n18 = MaterialFunctionCall(Float2=Saturate(None=Divide(A=@n17, B=@n8))) {materialFunction=&BreakOutFloat2Components}
  FunctionOutput(None=Max(A=@n18.R, B=@n18.R))
 ]}
