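# Default Lit, Nanite-enabled material that appears to be a debug view for inspecting Mesh Partition channel data.
# A MeshPartitionInspector node reads the channel texture from the MeshPartitionResource node.
# MF_PaintGrid builds a grid mask from unit-scaled partition UVs with a constant period.
# The mask blends the inspector result toward a color whose X and Y both come from the U coordinate.
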Material /MeshPartition/Materials/M_MeshPartitionInspector
graph {
 properties {bUsedWithNanite=true, usedShadingModels=Default Lit}
 nodes [
  MaterialExpressionMegaMeshResource_0 = MeshPartitionResource()
  MaterialExpressionMegaMeshInspector_0 = MeshPartitionInspector!(ChannelTextureInput=@MaterialExpressionMegaMeshResource_0.Channel Texture)
  MaterialExpressionMegaMeshTexcoord_0 = MeshPartitionTexcoord()
  n1 = MaterialFunctionCall(Float2=@MaterialExpressionMegaMeshTexcoord_0.uv) {materialFunction=&BreakOutFloat2Components}
  LinearInterpolate(
   A=@MaterialExpressionMegaMeshInspector_0
   B=MaterialFunctionCall(X=@n1.R, Y=@n1.R) {materialFunction=&MakeFloat3}.Result
   Alpha=MaterialFunctionCall!(Period=Constant() {r=1000}, Width, Signal=@MaterialExpressionMegaMeshTexcoord_0."uv [unit]") {materialFunction=&/MF_PaintGrid}.Result
  )
 ]
}
