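# Material function for MeshPartition that triplanar-projects a texture in world space, scaled by TextureSize.
# It samples the XZ, YZ and XY planes from world position (excluding shader offsets by default) using shared wrap samplers.
# Absolute normal X and Z, sharpened by inline CheapContrast subgraphs, blend the projections into XYZ, side-only and top-only outputs.
# A static switch optionally appends channels for float4 export.
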MaterialFunction /MeshPartition/Materials/MF_MeshPartition_WorldAlignedTexture
graph {nodes [
  n1 = FunctionInput(Preview=VertexNormalWS()) {bUsePreviewValueAsDefault=true, inputName=World Space Normal, previewValue=<0 0 1 1>}
  n2 = Abs(None=ComponentMask(None=@n1) {b=true})
  n3 = FunctionInput() {
   bUsePreviewValueAsDefault=true
   inputName=ProjectionTransitionContrast
   inputType=FunctionInput_Scalar
   sortPriority=10
   previewValue=<1 0 0 1>
  }
  Composite!(In=@n2, Contrast=@n3) {subgraphName=CheapContrast2}
  n4 = Abs(None=ComponentMask(None=@n1) {r=true})
  Composite(In=@n4, Contrast=@n3) {subgraphName=CheapContrast1}
  n5 = FunctionInput() {bUsePreviewValueAsDefault=true, description=Size of texture in worldspace units, inputName=TextureSize, previewValue=<64 64 64 6>}
  n6 = Divide(
   A=FunctionInput(Preview=WorldPosition() {worldPositionShaderOffset=WPT_ExcludeAllShaderOffsets}.XYZ) {bUsePreviewValueAsDefault=true, inputName=WorldPosition}
   B=Multiply(A=Abs(None=@n5)) {constB=-1}
  )
  n7 = FunctionInput(Preview=TextureObject() {texture=/Engine/EngineMaterials/DefaultWhiteGrid}) {inputName=TextureObject, inputType=FunctionInput_Texture2D}
  n8 = TextureSample(UVs=ComponentMask(None=@n6) {b=true, r=true}, Tex=@n7) {samplerSource=SSM_Wrap_WorldGroupSettings}
  n9 = FunctionInput(Preview=StaticBool()) {bUsePreviewValueAsDefault=true, description="Default = False", inputName=Export Float 4, inputType=FunctionInput_StaticBool}
  n10 = TextureSample(UVs=ComponentMask(None=@n6) {b=true, g=true}, Tex=@n7) {samplerSource=SSM_Wrap_WorldGroupSettings}
  MaterialExpressionReroute_5 = Reroute(None=@n3)
  n11 = Subtract(B=@MaterialExpressionReroute_5.Contrast) {constA=0, desc=Low negative value}
  n12 = Add(A=@MaterialExpressionReroute_5.Contrast) {desc=High Positive}
  MaterialExpressionReroute_0 = Reroute(None=@n4)
  n13 = ComponentMask(None=Clamp(None=LinearInterpolate(A=@n11, B=@n12, Alpha=@MaterialExpressionReroute_0.In) {desc=Lerps between greater range of value})) {r=true}
  MaterialExpressionReroute_7 = Reroute(None=@n13)
  n14 = LinearInterpolate(
   A=StaticSwitch(True=AppendVector(A=@n8.RGB, B=@n8.RGB), False=@n8.RGB, Value=@n9)
   B=StaticSwitch(True=AppendVector(A=@n10.RGB, B=@n10.RGB), False=@n10.RGB, Value=@n9)
   Alpha=ComponentMask(None=@MaterialExpressionReroute_7.Out) {r=true}
  ) {desc=Float3}
  n15 = TextureSample(UVs=ComponentMask(None=@n6) {g=true, r=true}, Tex=@n7) {samplerSource=SSM_Wrap_WorldGroupSettings}
  n16 = StaticSwitch(True=AppendVector(A=@n15.RGB, B=@n15.RGB), False=@n15.RGB, Value=@n9)
  MaterialExpressionReroute_2 = Reroute(None=@n3)
  n17 = Subtract(B=@MaterialExpressionReroute_2.Contrast) {constA=0, desc=Low negative value}
  n18 = Add(A=@MaterialExpressionReroute_2.Contrast) {desc=High Positive}
  MaterialExpressionReroute_1 = Reroute(None=@n2)
  n19 = ComponentMask(None=Clamp(None=LinearInterpolate(A=@n17, B=@n18, Alpha=@MaterialExpressionReroute_1.In) {desc=Lerps between greater range of value})) {r=true}
  MaterialExpressionReroute_4 = Reroute(None=@n19)
  FunctionOutput(None=LinearInterpolate(A=@n14, B=@n16, Alpha=ComponentMask(None=@MaterialExpressionReroute_4.Out) {r=true})) {outputName=XYZ Texture, sortPriority=1}
  FunctionOutput(None=@n14) {outputName=XY Texture, sortPriority=1}
  FunctionOutput(None=@n16) {outputName=Z Texture, sortPriority=1}
  PinBase() {reroutePins [
    In {expression=&self:MaterialExpressionReroute_1}
    Contrast {expression=&self:MaterialExpressionReroute_2}
   ]}
  PinBase!(Out=@n19) {reroutePins [Out {expression=&self:MaterialExpressionReroute_4}]}
  PinBase!() {reroutePins [
    In {expression=&self:MaterialExpressionReroute_0}
    Contrast {expression=&self:MaterialExpressionReroute_5}
   ]}
  PinBase!(Out=@n13) {reroutePins [Out {expression=&self:MaterialExpressionReroute_7}]}
 ]}
