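# Default Nanite-enabled lit material for Mesh Partition that tints a world-aligned grid by painted channels.
# MF_MeshPartition_WorldAlignedTexture projects the Texture_Color grid using the Texture_Size parameter.
# MeshPartitionResource supplies a channel texture, sampled as Red, Green, Blue and Yellow channels defined in MPD_Default.
# Named reroutes carry these into chained lerps that layer red, green, blue and orange-yellow tints over a grey base.
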Material /MeshPartition/Materials/M_MeshPartition_Default
graph {
 properties {
  bAutomaticallySetUsageInEditor=false
  bUsedWithNanite=true
  floatPrecisionMode=MFPM_Full_MaterialExpressionOnly
  usedShadingModels=Default Lit
 }
 nodes [
  n1 = LinearInterpolate(
   A=Multiply(A=Constant3Vector() {constant=<rgba 0.5 0.5 0.5 0>}, B=NamedRerouteUsage())
   B=Multiply(A=Constant3Vector() {constant=<rgba 0.5 0 0 0>}, B=NamedRerouteUsage())
   Alpha=NamedRerouteUsage()
  ) {desc=Lerp Red}
  n2 = Multiply(A=Constant3Vector() {constant=<rgba 0 0.5 0 0>}, B=NamedRerouteUsage())
  n3 = NamedRerouteUsage()
  n4 = Constant3Vector() {constant=<rgba 0 0 0.5 0>}
  n5 = NamedRerouteUsage()
  LinearInterpolate(
   A=LinearInterpolate(A=LinearInterpolate(A=@n1, B=@n2, Alpha=@n3) {desc=Lerp Green}, B=Multiply(A=@n4, B=@n5), Alpha=NamedRerouteUsage()) {desc=Lerp Blue}
   B=Multiply(A=Constant3Vector() {constant=<rgba 1 0.5 0 1>}, B=NamedRerouteUsage())
   Alpha=NamedRerouteUsage()
  ) {desc=Lerp Yellow}
  NamedRerouteDeclaration(None=MaterialFunctionCall!(
    TextureObject=TextureObjectParameter() {group=Textures, parameterName=Texture_Color, texture=/Engine/EngineMaterials/DefaultWhiteGrid}
    TextureSize=VectorParameter() {parameterName=Texture_Size, defaultValue=<rgba 200 200 200 0>}.RGB
    WorldPosition
    Export Float 4
    World Space Normal
    ProjectionTransitionContrast
   ) {materialFunction=&/MF_MeshPartition_WorldAlignedTexture}.XYZ Texture) {name=WorldAlignedGrid, nodeColor=<rgba 0.5 0.5 0.5 1>}
  n6 = MeshPartitionResource()
  NamedRerouteDeclaration(None=MeshPartitionChannelSample!(ChannelTextureInput=@n6.Channel Texture) {channel=Red, meshPartitionDefinition=/MeshPartition/DataAssets/MPD_Default}) {name=Red, nodeColor=<rgba 1 0 0 1>}
  NamedRerouteDeclaration(None=MeshPartitionChannelSample(ChannelTextureInput=@n6.Channel Texture) {channel=Green, meshPartitionDefinition=&MPD_Default}) {name=Green, nodeColor=<rgba 0 1 0 1>}
  NamedRerouteDeclaration(None=MeshPartitionChannelSample(ChannelTextureInput=@n6.Channel Texture) {channel=Blue, meshPartitionDefinition=&MPD_Default}) {name=Blue, nodeColor=<rgba 0 0 1 1>}
  NamedRerouteDeclaration(None=MeshPartitionChannelSample(ChannelTextureInput=@n6.Channel Texture) {channel=Yellow, meshPartitionDefinition=&MPD_Default}) {name=Yellow, nodeColor=<rgba 1 0.5 0 1>}
 ]
}
